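# Unlit alpha-composite surface material for the MimosaXP points icon, used in UI item previews.
# MF_UI_MimosaXP_Points_Base supplies color and mask, with a shine swipe color-dodged over the color, then multiplied by the mask and vertex alpha.
# A time-driven sine blends between parameterized bob offsets and a slight Z-axis rotation about the object position, summed into a vertex-style offset that makes the icon bob and sway.
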Material /Game/UI/Foundation/MimosaXP/ItemPreview_Materials/M_UI_Icon_MimosaXP_Points_Surface
graph {
 properties {bAutomaticallySetUsageInEditor=false, blendMode=BLEND_AlphaComposite, shadingModel=MSM_Unlit, usedShadingModels=Unlit}
 nodes [
  n1 = Sine(None=Multiply(A=Time(), B=ScalarParameter() {defaultValue=0.2, parameterName=BobRate}))
  n2 = Constant() {r=-0.01}
  n3 = Constant() {r=0.01}
  Add(
   A=LinearInterpolate(
    A=VectorParameter() {parameterName=BobHeightMax, defaultValue=<rgba 0 0 1 0>}.RGB
    B=VectorParameter() {parameterName=BobHeightMin}.RGB
    Alpha=@n1
   )
   B=RotateAboutAxis(
    NormalizedRotationAxis=Constant3Vector() {constant=<rgba 0 0 1 0>}
    RotationAngle=LinearInterpolate(A=@n2, B=@n3, Alpha=@n1)
    PivotPoint=ObjectPositionWS()
    Position=WorldPosition().XYZ
   )
  )
  n4 = NamedRerouteUsage()
  n5 = MaterialFunctionCall() {materialFunction=/Game/UI/Foundation/MimosaXP/MaterialFunction/MF_UI_MimosaXP_Points_Base}
  Multiply(
   A=VertexColor().A
   B=Multiply(
    A=MaterialFunctionCall!(Blend=@n4, Base=@n5.Color) {materialFunction=/Engine/Functions/Engine_MaterialFunctions03/Blends/Blend_ColorDodge}.Result
    B=@n5.Alpha Mask
   )
  )
  NamedRerouteDeclaration(None=Multiply(
    A=ScalarParameter() {defaultValue=1.5, parameterName=Shine-Multiply}
    B=MaterialFunctionCall!(In_Uv) {materialFunction=/Game/UI/Foundation/Materials/M-Function-ShineSwipe}.Alpha
   )) {name=ShineSwipe, nodeColor=<rgba 0 0 0 1>}
 ]
}
